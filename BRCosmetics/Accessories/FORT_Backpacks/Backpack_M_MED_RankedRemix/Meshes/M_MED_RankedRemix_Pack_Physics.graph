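# Physics asset for the MED Ranked Remix backpack cosmetic, defining secondary motion for its main piece, chains, tags, trapezoid and crown pivot.
# Small capsules form base/dyn bone pairs, mostly joined by angular limits that are tight on the main piece and looser on chains and tags.
# Chain and trapezoid bases attach to dyn_main, and tag bases to the front chain, with free rotation.
# The crown's dyn-to-base joint is locked, leaving limited twist against dyn_main.
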PhysicsAsset /BRCosmetics/Accessories/FORT_Backpacks/Backpack_M_MED_RankedRemix/Meshes/M_MED_RankedRemix_Pack_Physics
bodies {
 base_chain_bt [{shapeType=Capsule, length=1, radius=1, shapeName=base_chain_bt_capsule, rotation=<rot -90 0 0>}]
 base_chain_fr [{shapeType=Capsule, length=1, radius=1, shapeName=base_chain_fr_capsule, rotation=<rot -90 0 0>}]
 base_chain_l [{shapeType=Capsule, length=1, radius=1, shapeName=base_chain_l_capsule, rotation=<rot -90 0 0>}]
 base_chain_r [{shapeType=Capsule, length=1, radius=1, shapeName=base_crown_pivot_capsule, rotation=<rot -90 0 0>}]
 base_crown_pivot [{shapeType=Capsule, length=1, radius=1, shapeName=base_crown_pivot_capsule, rotation=<rot -90 0 0>}]
 base_main [{shapeType=Capsule, length=1, radius=1, shapeName=base_main_capsule, rotation=<rot -90 0 0>}]
 base_tag_l [{shapeType=Capsule, length=1, radius=1, shapeName=base_tag_l_capsule, rotation=<rot -90 0 0>}]
 base_tag_r [{shapeType=Capsule, length=1, radius=1, shapeName=base_tag_r_capsule, rotation=<rot -90 0 0>}]
 base_trapezoid [{shapeType=Capsule, length=1, radius=1, shapeName=base_trapezoid_capsule, rotation=<rot -90 0 0>}]
 dyn_chain_bt [{shapeType=Capsule, length=1.5, radius=1, shapeName=dyn_chain_bt_capsule, center=<1.5 0 0>, rotation=<rot -90 -173.5632 173.5632>}]
 dyn_chain_fr [{shapeType=Capsule, length=1.5, radius=1, shapeName=dyn_chain_fr_capsule, center=<1.5 0 0>, rotation=<rot -90 0 0>}]
 dyn_chain_l [{shapeType=Capsule, length=1.5, radius=1, shapeName=dyn_chain_l_capsule, center=<2 0 0>, rotation=<rot -90 90.4779 -90.4779>}]
 dyn_chain_r [{shapeType=Capsule, length=1.5, radius=1, shapeName=dyn_chain_l_capsule, center=<2 0 0>, rotation=<rot -90 -108.7956 108.7956>}]
 dyn_crown_pivot [{shapeType=Capsule, length=2, radius=1, shapeName=dyn_crown_pivot_capsule, center=<2 0 0>, rotation=<rot -90 0 0>}]
 dyn_main [{shapeType=Capsule, length=1, radius=1, shapeName=dyn_main_capsule, rotation=<rot -90 0 0>}]
 dyn_tag_l [{shapeType=Capsule, length=1, radius=1, shapeName=dyn_tag_l_capsule, center=<1 0 0>, rotation=<rot -90 0 0>}]
 dyn_tag_r [{shapeType=Capsule, length=1, radius=1, shapeName=dyn_tag_r_capsule, center=<1 0 0>, rotation=<rot -90 0 0>}]
 dyn_trapezoid [{shapeType=Capsule, length=2.5, radius=1, shapeName=dyn_trapezoid_capsule, center=<2 0 0>, rotation=<rot -90 0 0>}]
}
constraints [
 {bone1Name=dyn_trapezoid, bone2Name=base_trapezoid, swing1LimitDegrees=7.5, swing2LimitDegrees=2, twistLimitDegrees=1}
 {
  bone1Name=dyn_crown_pivot
  bone2Name=base_crown_pivot
  swing1LimitDegrees=5
  swing1Motion=Locked
  swing2LimitDegrees=5
  swing2Motion=Locked
  twistLimitDegrees=15
  twistMotion=Locked
 }
 {bone1Name=dyn_chain_l, bone2Name=base_chain_l, swing1LimitDegrees=22.5, swing2LimitDegrees=22.5, twistLimitDegrees=0}
 {bone1Name=dyn_tag_r, bone2Name=base_tag_r, swing1LimitDegrees=30, swing2LimitDegrees=15, twistLimitDegrees=5}
 {bone1Name=dyn_tag_l, bone2Name=base_tag_l, swing1LimitDegrees=40, swing2LimitDegrees=15, twistLimitDegrees=15}
 {
  bone1Name=base_tag_r
  bone2Name=dyn_chain_fr
  swing1LimitDegrees=45
  swing1Motion=Free
  swing2LimitDegrees=45
  swing2Motion=Free
  twistLimitDegrees=45
  twistMotion=Free
 }
 {
  bone1Name=base_tag_l
  bone2Name=dyn_chain_fr
  swing1LimitDegrees=45
  swing1Motion=Free
  swing2LimitDegrees=45
  swing2Motion=Free
  twistLimitDegrees=45
  twistMotion=Free
 }
 {bone1Name=dyn_chain_fr, bone2Name=base_chain_fr, swing1LimitDegrees=20, swing2LimitDegrees=22.5, twistLimitDegrees=10}
 {bone1Name=dyn_chain_bt, bone2Name=base_chain_bt, swing1LimitDegrees=22.5, swing2LimitDegrees=22.5, twistLimitDegrees=5}
 {
  bone1Name=base_trapezoid
  bone2Name=dyn_main
  swing1LimitDegrees=45
  swing1Motion=Free
  swing2LimitDegrees=45
  swing2Motion=Free
  twistLimitDegrees=45
  twistMotion=Free
 }
 {bone1Name=base_crown_pivot, bone2Name=dyn_main, swing1LimitDegrees=0, swing2LimitDegrees=0, twistLimitDegrees=15}
 {
  bone1Name=base_chain_bt
  bone2Name=dyn_main
  swing1LimitDegrees=45
  swing1Motion=Free
  swing2LimitDegrees=45
  swing2Motion=Free
  twistLimitDegrees=45
  twistMotion=Free
 }
 {
  bone1Name=base_chain_l
  bone2Name=dyn_main
  swing1LimitDegrees=45
  swing1Motion=Free
  swing2LimitDegrees=45
  swing2Motion=Free
  twistLimitDegrees=45
  twistMotion=Free
 }
 {
  bone1Name=base_chain_fr
  bone2Name=dyn_main
  swing1LimitDegrees=45
  swing1Motion=Free
  swing2LimitDegrees=45
  swing2Motion=Free
  twistLimitDegrees=45
  twistMotion=Free
 }
 {bone1Name=dyn_main, bone2Name=base_main, swing1LimitDegrees=5, swing2LimitDegrees=5, twistLimitDegrees=5}
 {
  bone1Name=base_chain_r
  bone2Name=dyn_main
  swing1LimitDegrees=25
  swing1Motion=Free
  swing2LimitDegrees=45
  swing2Motion=Free
  twistLimitDegrees=10
  twistMotion=Free
 }
 {bone1Name=dyn_chain_r, bone2Name=base_chain_r, swing1LimitDegrees=20, swing2LimitDegrees=27.5, twistLimitDegrees=5}
]
